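# Translucent UI material for the survivor tiers screens that draws a glowing diamond outline over an animated radial shine burst.
# The shine sums two blurred noise samples panned at different speeds over radial angle coordinates, then masks them with an exponential radial gradient.
# The diamond is an MF_UI_Box outline on UVs rotated an eighth turn.
# Both layers are added as separately tinted color and as an untinted mask, with parameters for shape and color.
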Material /Game/UI/SurvivorTiers/Materials/M_UI_DiamondBurst
graph {
 properties {
  bUsedWithStaticMesh=false
  blendMode=BLEND_Translucent
  materialDomain=MD_UI
  pixelDepthOffsetMode=PDOM_Legacy
  usedShadingModels=Default Lit
 }
 nodes [
  n1 = MaterialFunctionCall() {materialFunction=&VectorToRadialValue}
  n2 = Multiply(A=Panner(Coordinate=@n1.Vector Converted to Angle) {speedX=0.15}) {constB=2}
  n3 = Multiply(A=Panner(Coordinate=@n1.Vector Converted to Angle) {speedX=-0.125, speedY=-0.125}) {constB=2}
  n4 = Multiply(
   A=Add(
    A=TextureSample(UVs=@n2) {samplerType=SAMPLERTYPE_Grayscale, texture=&LowResBlurredNoise}.R
    B=TextureSample(UVs=@n3) {samplerType=SAMPLERTYPE_Grayscale, texture=&LowResBlurredNoise}.R
   )
   B=MaterialFunctionCall(
    Radius=ScalarParameter() {defaultValue=0.5, group=Shine, parameterName=ShineRadius, sortPriority=1}
    Density=ScalarParameter() {defaultValue=1.5, group=Shine, parameterName=ShineAlpha, sortPriority=2}
   ) {materialFunction=&RadialGradientExponential}.RadialGradientExponential
  ) {constB=2}
  n5 = MaterialFunctionCall(
   UVs=MaterialFunctionCall(UVs=TextureCoordinate(), "Rotation Angle (0-1)"=Constant() {r=-0.125}) {materialFunction=&CustomRotator}.Rotated Values
   Size=ScalarParameter() {defaultValue=0.5, group=Diamond, parameterName=DiamondSize, sortPriority=1}
   CornerRoundness=Constant()
   OutlineThickness=ScalarParameter() {defaultValue=0.05, group=Diamond, parameterName=DiamondThickness, sortPriority=2}
   GlowMax=ScalarParameter() {group=Diamond, parameterName=DiamondGlow, sortPriority=3}
  ) {materialFunction=&MF_UI_Box}
  Add(
   A=Multiply(A=VectorParameter() {group=Shine, parameterName=ShineColor, sortPriority=4, defaultValue=<rgba 1 0.9823 0.7991 1>}.RGB, B=@n4)
   B=Multiply(A=VectorParameter() {group=Diamond, parameterName=DiamondColor, sortPriority=5, defaultValue=<rgba 1 0.9823 0.7991 1>}.RGB, B=@n5.Outline)
  )
  Add(A=@n4, B=@n5.Outline)
 ]
}
