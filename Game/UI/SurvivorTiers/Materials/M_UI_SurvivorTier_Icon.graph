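# UI material for survivor tier icons, drawing a tinted seasonal icon over a tier icon texture.
# Seasonal texture UVs take X/Y offsets and scale about center via ScaleUVsByCenter.
# Its red channel, multiplied by IconColor, also masks the blend over the tier color, while the tier alpha is passed through separately.
# The UseSmallTextures lerp has identical inputs, so that parameter has no visible effect.
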Material /Game/UI/SurvivorTiers/Materials/M_UI_SurvivorTier_Icon
graph {
 properties {
  bUsedWithStaticMesh=false
  blendMode=BLEND_Translucent
  materialDomain=MD_UI
  pixelDepthOffsetMode=PDOM_Legacy
  usedShadingModels=Default Lit
 }
 nodes [
  n1 = TextureSample(Tex=TextureObjectParameter() {parameterName=TierIcon, texture=/Game/UI/SurvivorTiers/Textures/S-Tiers-New_T_UI_SurvivalTierIcon_Tier7})
  n2 = TextureCoordinate()
  n3 = Add(A=ScalarParameter() {defaultValue=-0.02, parameterName=XOffset}, B=ComponentMask(None=@n2) {r=true})
  n4 = Add(A=ComponentMask(None=@n2) {g=true}, B=ScalarParameter() {defaultValue=-0.032, parameterName=YOffset})
  n5 = TextureSampleParameter2D(UVs=MaterialFunctionCall(UVs=AppendVector(A=@n3, B=@n4), Texture Scale=ScalarParameter() {defaultValue=0.4, parameterName=TextureScale}) {materialFunction=&ScaleUVsByCenter}.UVs) {parameterName=SeasonalIcon, texture=/Game/Athena/UI/Frontend/Spatial/Art/T_Ui_Empty}
  n6 = LinearInterpolate(A=@n5.R, B=@n5.R, Alpha=ScalarParameter() {parameterName=UseSmallTextures})
  LinearInterpolate(A=@n1.RGB, B=Multiply(A=@n6, B=VectorParameter() {parameterName=IconColor, defaultValue=<rgba 0.9047 0.8308 0.552 1>}.RGB), Alpha=@n6)
  Reroute(None=@n1.A)
 ]
}
