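# Sound cue for the post-game tier badge slam in the UI, played through the Fort_UI sound class.
# A single wave player is the root and plays the UI_SurvivorBadge_Slam wave, with no random or modulation nodes.
# The cue raises volume slightly and overrides concurrency with a short retrigger time.
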SoundCue /Game/Sounds/Fort_UI_Sounds/Gameplay/PostGame/Fortnite_UI_PostGame_TierBadge_Slam_Cue
root=WavePlayer_1
nodes [{id=WavePlayer_1, properties {soundWaveAssetPtr=/Game/Sounds/Fort_UI_Sounds/Gameplay/PostGame/wav/UI_SurvivorBadge_Slam}}]
properties {
 bOverrideConcurrency=true
 maxDistance=2097152
 soundClassObject=&Fort_UI
 volumeMultiplier=1.25
 concurrencyOverrides {retriggerTime=0.15}
}
