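# Sound cue for the whoosh of the tier badge on the post-game UI screen.
# A single wave player feeds the SurvivorBadge whoosh wave straight to the output, with boosted volume on the Fort UI sound class.
# A concurrency override adds a short retrigger time.
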SoundCue /Game/Sounds/Fort_UI_Sounds/Gameplay/PostGame/Fortnite_UI_PostGame_TierBadge_Whoosh_Cue
root=WavePlayer_0
nodes [{id=WavePlayer_0, properties {soundWaveAssetPtr=/Game/Sounds/Fort_UI_Sounds/Gameplay/PostGame/wav/UI_SurvivorBadge_Whoosh}}]
properties {
 bOverrideConcurrency=true
 maxDistance=2097152
 soundClassObject=&Fort_UI
 volumeMultiplier=4
 concurrencyOverrides {retriggerTime=0.15}
}
