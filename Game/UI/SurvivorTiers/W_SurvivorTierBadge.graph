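# Survivor tier badge widget that shows a prestige icon and tier number in small or large size, styled from seasonal display data.
# For swap presentation it shows the before state, async-preloads the next prestige material and tints the diamond burst.
# MedalSwapAnimation drives shine, scale and burst, with event tracks for whoosh and slam sounds and the visual update.
# While awaiting a swap, state changes feed the stored target instead of redrawing.
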WidgetBlueprint /Game/UI/SurvivorTiers/W_SurvivorTierBadge
animations [MedalSwapAnimation {
  binding_count=3
  length_seconds=2.7667
  object=&self:MedalSwapAnimation
  sequence {
   bindings [
    Text_Tier {
     id=#1
     possessed_class=CommonTextBlock
     tracks [
      {
       type=MovieSceneFloatTrack
       display_name=Render Opacity
       sections [{
         type=MovieSceneFloatSection
         end_frame=32
         start_frame=0
         channels [None {type=MovieSceneScriptingFloatChannel, thinned_from=6, keys [[0,1],[17,1],[23,0],[27.5873,0],[32.9943,1]]}]
        }]
      }
      {
       type=MovieScene2DTransformTrack
       sections [{
         type=MovieScene2DTransformSection
         end_frame=37
         start_frame=0
         channels [
          Scale.X {type=MovieSceneScriptingFloatChannel, keys [[17,1],[24.9293,0.8],[25.2647,0.6],[27.5287,1.3],[32.301,1]]}
          Scale.Y {type=MovieSceneScriptingFloatChannel, keys [[17,1],[24.9293,0.8],[25.2647,0.6],[27.5287,1.3],[32.301,1]]}
         ]
        }]
      }
     ]
    }
    LazyImage_Icon {
     id=#2
     possessed_class=FortLazyImage
     tracks [
      {
       type=MovieSceneWidgetMaterialTrack
       sections [{
         type=MovieSceneParameterSection
         end_frame=55
         start_frame=0
         channels [
          ShineOffset {type=MovieSceneScriptingFloatChannel, keys [[0,1],[3,1],[18,-1],[24.9723,-1],[25.2647,1],[40.3593,1],[55.334,-1]]}
          WhiteShine {type=MovieSceneScriptingFloatChannel, keys [[24.9723,0],[25.613,1],[29.9943,0]]}
          ShineStrenght {type=MovieSceneScriptingFloatChannel, keys [[0,1],[3,1],[7,0.3],[14,0.5],[18,1],[40.3593,1],[44,0.3],[53,0.5],[55.334,0.9451]]}
         ]
        }]
      }
      {
       type=MovieSceneFloatTrack
       display_name=Render Opacity
       sections [{
         type=MovieSceneFloatSection
         end_frame=31
         start_frame=0
         channels [None {type=MovieSceneScriptingFloatChannel, keys [[0,1],[17,1],[24.9723,0],[25.2647,0],[27.325,1]]}]
        }]
      }
      {
       type=MovieScene2DTransformTrack
       sections [{
         type=MovieScene2DTransformSection
         end_frame=37
         start_frame=0
         channels [
          Scale.X {type=MovieSceneScriptingFloatChannel, keys [[17,1],[24,0.95],[25,0.2],[28,1.3],[32.301,1]]}
          Scale.Y {type=MovieSceneScriptingFloatChannel, keys [[17,1],[24,0.95],[25,0.2],[28,1.3],[32.301,1]]}
         ]
        }]
      }
     ]
    }
    Image_SwapDiamondBurst {
     id=#3
     possessed_class=Image
     tracks [
      {
       type=MovieSceneWidgetMaterialTrack
       sections [{
         type=MovieSceneParameterSection
         end_frame=45
         start_frame=0
         channels [
          ShineAlpha {type=MovieSceneScriptingFloatChannel, keys [[0,0],[15.5743,0],[23.5743,2],[27.5743,1],[32.5743,0]]}
          ShineRadius {type=MovieSceneScriptingFloatChannel, keys [[0,0],[17.5743,0],[23.5743,0.2346],[27.5743,0.5],[33,0.6]]}
          DiamondSize {type=MovieSceneScriptingFloatChannel, keys [[0,0],[22,0],[33,0.4],[41,0.45]]}
          DiamondThickness {type=MovieSceneScriptingFloatChannel, const [0,4]}
          DiamondGlow {type=MovieSceneScriptingFloatChannel, keys [[0,0],[28,0],[33,0.25],[41,0.1]]}
         ]
        }]
      }
      {
       type=MovieSceneFloatTrack
       display_name=Render Opacity
       sections [{
         type=MovieSceneFloatSection
         end_frame=45
         start_frame=0
         channels [None {type=MovieSceneScriptingFloatChannel, keys [[0,1],[28,1],[33,0.6],[41,0]]}]
        }]
      }
     ]
    }
   ]
   tracks [
    {
     type=MovieSceneEventTrack
     sections [{
       type=MovieSceneEventTriggerSection
       channels [None {type=MovieSceneScriptingEventChannel, const ["<Struct 'MovieSceneEvent' (0x4687ddcf500) {}>", 1]}]
      }]
    }
    {
     type=MovieSceneEventTrack
     display_name=Audio Events
     sections [{
       type=MovieSceneEventTriggerSection
       channels [None {
         type=MovieSceneScriptingEventChannel
         keys [[6, "<Struct 'MovieSceneEvent' (0x4687ddcf0a0) {}>"], [20, "<Struct 'MovieSceneEvent' (0x4687ddcaf00) {}>"]]
        }]
      }]
    }
   ]
  }
 }]
graphs [
 EventGraph {nodes [
   Delay(Duration=2)
   > n1 = Increment Int(Value=Get SwapAnim_AfterPrestigeLevel() -> (SwapAnim_AfterPrestigeLevel:int=0))
   > Branch(Condition="integer < integer"(A=@n1, B=7))
   > n2 = PreloadPrestigeMaterial!(execute, self, PrestigeLevel:int=@n1) -> (then)
   "Play Animation with Finished event (legacy)"!(
    execute:Exec=[@n3.then, @n2.then]
    Widget:User Widget Object Reference=Self-Reference()
    InAnimation:Widget Animation Object Reference=Get MedalSwapAnimation() -> (MedalSwapAnimation:Widget Animation Object Reference)
    StartAtTime:float=0
    NumLoopsToPlay:int=1
    PlayMode:EUMGSequencePlayMode Enum
    PlaybackSpeed:float=1
   ) -> (then:Exec, Result:UMGSequence Player Object Reference, Finished:Exec) {is=title K2Node_PlayAnimation}
   PlaySwapAnimation() {is=title K2Node_CustomEvent}
   > n3 = SetVisibility(self=Get Image_SwapDiamondBurst() -> (Image_SwapDiamondBurst:Image Object Reference), InVisibility=SelfHitTestInvisible)
   n4 = PreloadPrestigeMaterial!() -> (OutputDelegate, then, PrestigeLevel:int) {is=title K2Node_CustomEvent}
   n5 = GetSurvivorTierDisplayData!() -> (ReturnValue:Athena Seasonal Survivor Tiers Display Data Object Reference)
   > Is Valid(InputObject=@n5)
   > n6 = GetDynamicMaterial(self=Get Image_SwapDiamondBurst())
   n7 = Break Survivor Presige Data!(SurvivorPresigeData:"Survivor Presige Data Structure (by ref)"=GetPrestigeData!(self:Athena Seasonal Survivor Tiers Display Data Object Reference=@n5, PrestigeLevel:int=@n4.PrestigeLevel) -> (ReturnValue:Survivor Presige Data Structure)) -> (
    SmallIconMaterial:Material Interface Soft Object Reference
    LargeIconMaterial:Material Interface Soft Object Reference
    Color1:Linear Color Structure
    Color2:Linear Color Structure
    TextColor:Linear Color Structure
   ) {is=title K2Node_BreakStruct}
   > SetVectorParameterValue(self=@n6, ParameterName=DiamondColor, Value=@n7.Color1)
   > SetVectorParameterValue(self=@n6, ParameterName=ShineColor, Value=@n7.Color2)
   > n8 = Async Load Asset(Asset=Select!(
     Small:Material Interface Soft Object Reference=@n7
     Large:Material Interface Soft Object Reference=@n7.LargeIconMaterial
     Index:EFortAthenaSurvivorTierBadgeSize Enum=Get BadgeSize() -> (BadgeSize:EFortAthenaSurvivorTierBadgeSize Enum=Small)
    ) -> (ReturnValue:Material Interface Soft Object Reference))
   > n9 = Cast To MaterialInterface!(execute, Object=@n8) -> (then, CastFailed, AsMaterial Interface:Material Interface Object Reference) {is=title K2Node_DynamicCast}
   > Set SwapAnim_LoadedNewTier(SwapAnim_LoadedNewTier:Material Interface Object Reference=@n9) -> (Output_Get:Material Interface Object Reference)
   n10 = Event OnPresentAnimation!() -> (OutputDelegate, then, BeforeTier:int=0, BeforePrestigeLevel:int=0, AfterTier:int=0, AfterPrestigeLevel:int=0) {is=title K2Node_Event}
   > Set IsWaitingForSwapAnimation(IsWaitingForSwapAnimation:bool=true) -> (Output_Get:bool)
   > UpdateTierVisuals!(execute, self, ProfileTier:int=@n10.BeforeTier, PrestigeLevel:int=@n10.BeforePrestigeLevel) -> (then)
   > Set SwapAnim_AfterTier(SwapAnim_AfterTier:int=@n10.AfterTier) -> (Output_Get:int=0)
   > n11 = Set SwapAnim_AfterPrestigeLevel(SwapAnim_AfterPrestigeLevel:int=@n10.AfterPrestigeLevel) -> (Output_Get:int=0)
   > PreloadPrestigeMaterial(PrestigeLevel=@n11) {is=title K2Node_CallFunction}
   n12 = Event PreConstruct()
   > Parent: PreConstruct(IsDesignTime=@n12.IsDesignTime)
   > SetVisibility(
    self=Get Text_Tier() -> (Text_Tier:Common Text Object Reference)
    InVisibility=Select!(Option 0:ESlateVisibility Enum=Collapsed, Option 1:ESlateVisibility Enum=SelfHitTestInvisible, Index:bool=Get ShowTier() -> (ShowTier:bool)) -> (ReturnValue:ESlateVisibility Enum)
   )
   n13 = Event OnSurvivorTierStateChanged!() -> (OutputDelegate, then, Tier:int=0, PrestigeLevel:int=0) {is=title K2Node_Event}
   > n14 = Branch(Condition=Get IsWaitingForSwapAnimation() -> (IsWaitingForSwapAnimation:bool))
   > Set SwapAnim_AfterTier(SwapAnim_AfterTier=@n13.Tier) -> (Output_Get=0)
   > Branch(Condition="Not Equal (Integer)"(A=Get SwapAnim_AfterPrestigeLevel() -> (SwapAnim_AfterPrestigeLevel=0), B=@n13.PrestigeLevel))
   > n15 = Set SwapAnim_AfterTier(SwapAnim_AfterTier=@n13.PrestigeLevel) -> (Output_Get=0)
   > PreloadPrestigeMaterial(PrestigeLevel=@n15)
   UpdateTierVisuals(execute=@n14.else, ProfileTier=@n13.Tier, PrestigeLevel=@n13.PrestigeLevel)
   Event OnBadgeSizeChanged() {is=title K2Node_Event}
   n16 = Select!(Small:float=88, Large:float=384, Index:EFortAthenaSurvivorTierBadgeSize Enum=Get BadgeSize() -> (BadgeSize=Small)) -> (ReturnValue:float)
   n17 = Get LazyImage_Icon() -> (LazyImage_Icon:Fort Lazy Image Object Reference)
   > n18 = Set members in Slate Brush!(
    execute:Exec
    StructRef:"Slate Brush Structure (by ref)"=Get Brush!(self:Image Object Reference=@n17) -> (Brush:Slate Brush Structure)
    ImageSize:Vector 2D Structure=MakeVector2D(X=@n16, Y=@n16)
   ) -> (then:Exec, StructOut:"Slate Brush Structure (by ref)") {is=title K2Node_SetFieldsInStruct}
   > SetBrush(self=@n17, InBrush=@n18)
   > SetFont(
    self=Get Text_Tier()
    InFontInfo=Select(
     Small:Slate Font Info Structure=Get SmallFont() -> (SmallFont:Slate Font Info Structure)
     Large:Slate Font Info Structure=Get LargeFont() -> (LargeFont:Slate Font Info Structure)
     Index=Get BadgeSize() -> (BadgeSize=Small)
    ) -> (ReturnValue:Slate Font Info Structure)
   )
  ]}
 Sequencer Events {nodes [
   SequenceEvent_UpdateVisuals() {is=title K2Node_CustomEvent}
   > UpdateTierVisuals(
    ProfileTier=Get SwapAnim_AfterTier() -> (SwapAnim_AfterTier:int=0)
    PrestigeLevel=Get SwapAnim_AfterPrestigeLevel() -> (SwapAnim_AfterPrestigeLevel=0)
   )
   PlayShimmer() {is=title K2Node_CustomEvent}
   > SpawnSound2D!(
    execute
    Sound:Sound Base Object Reference=Get WhooshSound() -> (WhooshSound:Sound Base Object Reference)
    VolumeMultiplier:float=1
    PitchMultiplier:float=1
    StartTime:float=0
    ConcurrencySettings:Sound Concurrency Object Reference
    bPersistAcrossLevelTransition:bool=false
    bAutoDestroy:bool=true
   ) -> (then, ReturnValue:Audio Component Object Reference)
   PlaySlam() {is=title K2Node_CustomEvent}
   > SpawnSound2D(
    Sound=Get SlamSound() -> (SlamSound:Sound Base Object Reference)
    VolumeMultiplier=1
    PitchMultiplier=1
    StartTime=0
    bPersistAcrossLevelTransition=false
    bAutoDestroy=true
   )
  ]}
 UpdateTierVisuals {nodes [
   n1 = UpdateTierVisuals!() -> (then, ProfileTier:int, PrestigeLevel:int) {is=title K2Node_FunctionEntry}
   > "SetText (Text)"(self=Get Text_Tier(), InText="To Text (Integer)"(Value=@n1))
   n2 = GetSurvivorTierDisplayData()
   > Is Valid(InputObject=@n2)
   n3 = Break Survivor Presige Data!(SurvivorPresigeData=GetPrestigeData(self=@n2, PrestigeLevel=@n1.PrestigeLevel)) -> (SmallIconMaterial, LargeIconMaterial, TextColor)
   > SetBrushFromSoftMaterial!(
    execute
    self:Image Object Reference=Get LazyImage_Icon()
    SoftMaterial:Material Interface Soft Object Reference=Select(
     Small:Material Interface Soft Object Reference=@n3
     Large:Material Interface Soft Object Reference=@n3.LargeIconMaterial
     Index=Get BadgeSize() -> (BadgeSize=Small)
    ) -> (ReturnValue:Material Interface Soft Object Reference)
   ) -> (then)
   > SetColorAndOpacity(self=Get Text_Tier(), InColorAndOpacity=Make Slate Color(SpecifiedColor=@n3.TextColor))
  ]}
]
variables [
 IsWaitingForSwapAnimation {type=boolean}
 LargeFont {
  type=SlateFontInfo
  default {
   bForceMonospaced=false
   bMaterialIsStencil=false
   fontMaterial=null
   fontObject=&Burbank
   letterSpacing=0
   monospacedWidth=1
   size=37.5
   skewAmount=0
   typefaceFontName=BurbankRegular-Black
   outlineSettings {
    bApplyOutlineToDropShadows=false
    bMiteredCorners=false
    bSeparateFillAlpha=false
    outlineMaterial=null
    outlineSize=0
    outlineColor=<rgba 0 0 0 1>
   }
  }
 }
 ShowTier {type=boolean, default=true}
 SlamSound {type=SoundBase, default=/Game/Sounds/Fort_UI_Sounds/Gameplay/PostGame/Fortnite_UI_PostGame_TierBadge_Slam_Cue}
 SmallFont {
  type=SlateFontInfo
  default {
   bForceMonospaced=false
   bMaterialIsStencil=false
   fontMaterial=null
   fontObject=&BurbankSmall
   letterSpacing=0
   monospacedWidth=1
   size=10.5
   skewAmount=0
   typefaceFontName=Black
   outlineSettings {
    bApplyOutlineToDropShadows=false
    bMiteredCorners=false
    bSeparateFillAlpha=false
    outlineMaterial=null
    outlineSize=0
    outlineColor=<rgba 0 0 0 1>
   }
  }
 }
 SwapAnim_AfterPrestigeLevel {type=integer}
 SwapAnim_AfterTier {type=integer}
 SwapAnim_LoadedNewTier {type=MaterialInterface, default=null}
 WhooshSound {type=SoundBase, default=/Game/Sounds/Fort_UI_Sounds/Gameplay/PostGame/Fortnite_UI_PostGame_TierBadge_Whoosh_Cue}
]
widgets {
 info {inheritedWidgetCount=0, parentClass=/Script/FortniteUI.FortAthenaSurvivorTierBadge, rootWidgetClass=&Overlay, widgetCount=8}
 widgets [
  Overlay_0 {widgetClassPath=&Overlay, widget=Overlay_0}
  RadialDropShadow {
   widgetClassPath=&Image
   bIsVariable=true
   parent=Overlay_0
   widget=RadialDropShadow
   properties {brush {
     bIsDynamicallyLoaded=false
     drawAs=Image
     imageType=NoImage
     mirroring=NoMirror
     resourceName=None
     resourceObject=/Game/UI/Frontend/Lobby/Habanero/Habanero_Materials/MaterialInstances/Background/MI_UI_Habanero_RadialDropShadow_Black30
     tiling=NoTile
     imageSize=<32 32>
     margin=<ltrb 0 0 0 0>
     outlineSettings {
      bUseBrushTransparency=false
      roundingType=HalfHeightRadius
      width=0
      color {colorUseRule=UseColor_Specified, specifiedColor=<rgba 0 0 0 0>}
      cornerRadii=<0 0 0 0>
     }
     tintColor {colorUseRule=UseColor_Specified, specifiedColor=<rgba 1 1 1 1>}
     uVRegion {bIsValid=false, max=<0 0>, min=<0 0>}
    }}
   slot {class=OverlaySlot, properties {horizontalAlignment=HAlign_Fill, verticalAlignment=VAlign_Fill}}
  }
  LazyImage_Icon {
   widgetClassPath=&FortLazyImage
   bIsVariable=true
   parent=Overlay_0
   widget=LazyImage_Icon
   properties {brush {
     bIsDynamicallyLoaded=false
     drawAs=Image
     imageType=NoImage
     mirroring=NoMirror
     resourceName=None
     resourceObject=null
     tiling=NoTile
     imageSize=<88 88>
     margin=<ltrb 0 0 0 0>
     outlineSettings {
      bUseBrushTransparency=false
      roundingType=HalfHeightRadius
      width=0
      color {colorUseRule=UseColor_Specified, specifiedColor=<rgba 0 0 0 0>}
      cornerRadii=<0 0 0 0>
     }
     tintColor {colorUseRule=UseColor_Specified, specifiedColor=<rgba 1 1 1 1>}
     uVRegion {bIsValid=false, max=<0 0>, min=<0 0>}
    }}
   slot {class=OverlaySlot, properties {horizontalAlignment=HAlign_Fill, verticalAlignment=VAlign_Fill}}
  }
  VerticalBox_PositionHack {
   widgetClassPath=&VerticalBox
   parent=Overlay_0
   widget=VerticalBox_PositionHack
   properties {renderTransform {angle=0, scale=<1 1>, shear=<0 0>, translation=<0 -1>}}
   slot {class=OverlaySlot, properties {horizontalAlignment=HAlign_Center, verticalAlignment=VAlign_Fill}}
  }
  Text_Tier {
   widgetClassPath=&CommonTextBlock
   bIsVariable=true
   parent=VerticalBox_PositionHack
   widget=Text_Tier
   properties {
    text=0
    font {
     bForceMonospaced=false
     bMaterialIsStencil=false
     fontMaterial=null
     fontObject=&Burbank
     letterSpacing=0
     monospacedWidth=1
     size=12
     skewAmount=0
     typefaceFontName=BurbankRegular-Black
     outlineSettings {
      bApplyOutlineToDropShadows=false
      bMiteredCorners=false
      bSeparateFillAlpha=true
      outlineMaterial=&M-FontShadow
      outlineSize=2
      outlineColor=<rgba 0 0 0 0.25>
     }
    }
    shadowOffset=<0 0>
   }
   slot {class=VerticalBoxSlot, properties {verticalAlignment=VAlign_Center, size {sizeRule=Fill, value=0.23}}}
  }
  Spacer_75 {
   widgetClassPath=&Spacer
   parent=VerticalBox_PositionHack
   widget=Spacer_75
   slot {class=VerticalBoxSlot, properties {size {sizeRule=Fill, value=1}}}
  }
  FortVisualAttachment_0 {
   widgetClassPath=&FortVisualAttachment
   parent=Overlay_0
   widget=FortVisualAttachment_0
   properties {contentAnchor=<0.5 0.5>}
   slot {class=OverlaySlot, properties {horizontalAlignment=HAlign_Center, verticalAlignment=VAlign_Center}}
  }
  Image_SwapDiamondBurst {
   widgetClassPath=&Image
   bIsVariable=true
   parent=FortVisualAttachment_0
   widget=Image_SwapDiamondBurst
   properties {
    visibility=Collapsed
    brush {
     bIsDynamicallyLoaded=false
     drawAs=Image
     imageType=NoImage
     mirroring=NoMirror
     resourceName=None
     resourceObject=/Game/UI/SurvivorTiers/Materials/MaterialInstances/MI_UI_DiamondBurst
     tiling=NoTile
     imageSize=<750 750>
     margin=<ltrb 0 0 0 0>
     outlineSettings {
      bUseBrushTransparency=false
      roundingType=HalfHeightRadius
      width=0
      color {colorUseRule=UseColor_Specified, specifiedColor=<rgba 0 0 0 0>}
      cornerRadii=<0 0 0 0>
     }
     tintColor {colorUseRule=UseColor_Specified, specifiedColor=<rgba 1 1 1 1>}
     uVRegion {bIsValid=false, max=<0 0>, min=<0 0>}
    }
   }
   slot {class=SizeBoxSlot}
  }
 ]
}
